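# Editor material for Dataflow that shows vertex colors on two-sided meshes, including skeletal meshes, with a dull red tint on back faces.
# It fakes lighting from camera-to-normal dot products, adding a tight vertex-colored rim to a view-facing term.
# A FlatShading parameter blends pixel normals toward faceted world-position derivative normals, disabled in orthographic views to avoid artifacts.
# EyeAdaptationInverse cancels exposure on the combined color.
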Material /Engine/EditorMaterials/Dataflow/DataflowTwoSidedVertexMaterial
graph {
 properties {
  bCastRayTracedShadows=false
  bUseTranslucencyVertexFog=false
  bUsedWithMeshDeformer=true
  bUsedWithSkeletalMesh=true
  blendMode=BLEND_Masked
  pixelDepthOffsetMode=PDOM_Legacy
  translucencyLightingMode=TLM_VolumetricDirectional
  twoSided=true
  usedShadingModels=Default Lit
 }
 nodes [
  Constant() {r=1}
  Constant() {r=0.6}
  Constant()
  n1 = WorldPosition() {desc=Flat Normal}
  n2 = DDX(Value=@n1.XYZ) {desc=Flat Normal}
  n3 = DDY(Value=@n1.XYZ) {desc=Flat Normal}
  n4 = OneMinus(None=IsOrthographic() {desc="DDX/DDY produces artifacts\r\nin Ortho views"})
  n5 = ScalarParameter() {defaultValue=1, desc=Flat Shading, parameterName=FlatShading, sliderMax=1}
  n6 = LinearInterpolate(
   A=PixelNormalWS()
   B=Normalize(VectorInput=CrossProduct(A=@n2, B=@n3) {desc=Flat Normal}) {desc=Flat Normal}
   Alpha=Clamp(None=Multiply(A=@n4, B=@n5))
  )
  n7 = CameraVectorWS()
  n8 = Clamp(None=DotProduct(A=@n6, B=@n7))
  n9 = Clamp(None=DotProduct(A=@n7, B=@n6))
  Divide(A=Add(A=@n8, B=@n9))
  n10 = Multiply(A=Power(Base=OneMinus(None=@n8)) {constExponent=12})
  n11 = VertexColor()
  n12 = Power(Base=@n9) {constExponent=3}
  n13 = If(A=TwoSidedSign(), "A > B"=@n11, "A == B"=@n11, "A < B"=Constant3Vector() {desc=Back, constant=<rgba 0.6 0.2 0.2 0>})
  EyeAdaptationInverse(LightValueInput=Add(A=Multiply(A=@n10, B=@n11), B=Multiply(A=@n12, B=@n13))) {desc=Undo effect of exposure}
 ]
}
