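# Editor material for Dataflow, named for vertex display, set up as two-sided Default Lit.
# It is flagged for use with skeletal meshes and mesh deformers.
# The graph holds a grey vector parameter and a Vertex Color node, with no connections shown between them or to material outputs.
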Material /Engine/EditorMaterials/Dataflow/DataflowVertexMaterial
graph {
 properties {bUsedWithMeshDeformer=true, bUsedWithSkeletalMesh=true, pixelDepthOffsetMode=PDOM_Legacy, twoSided=true, usedShadingModels=Default Lit}
 nodes [
  VectorParameter() {defaultValue=<rgba 0.5 0.5 0.5 0>}
  VertexColor()
 ]
}
