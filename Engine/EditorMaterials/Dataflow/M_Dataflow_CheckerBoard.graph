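# Editor material for Dataflow that shows a tintable checkerboard on a chosen UV channel, likely for inspecting UV layouts.
# A UVChannel scalar parameter drives a Switch node that selects one of eight texture coordinate sets, all with the same tiling.
# The selected UVs sample T_CheckerBoard, which is multiplied by the Color vector parameter.
# The material uses Default Lit shading and is flagged for instanced static meshes and Nanite.
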Material /Engine/EditorMaterials/Dataflow/M_Dataflow_CheckerBoard
graph {
 expressions {
  MaterialExpressionMultiply_0 {
   class=MaterialExpressionMultiply
   inputs [
    A {from=MaterialExpressionVectorParameter_0, output=RGB}
    B {from=MaterialExpressionTextureSample_0, output=RGB}
   ]
  }
  MaterialExpressionScalarParameter_1 {class=MaterialExpressionScalarParameter, properties {parameterName=UVChannel, sliderMax=7}}
  MaterialExpressionSwitch_0 {
   class=MaterialExpressionSwitch
   inputs [
    SwitchValue {from=MaterialExpressionScalarParameter_1, output=""}
    Default {from="", output=""}
    None {from=MaterialExpressionTextureCoordinate_0, output=""}
    None {from=MaterialExpressionTextureCoordinate_1, output=""}
    None {from=MaterialExpressionTextureCoordinate_2, output=""}
    None {from=MaterialExpressionTextureCoordinate_3, output=""}
    None {from=MaterialExpressionTextureCoordinate_4, output=""}
    None {from=MaterialExpressionTextureCoordinate_5, output=""}
    None {from=MaterialExpressionTextureCoordinate_6, output=""}
    None {from=MaterialExpressionTextureCoordinate_7, output=""}
   ]
   properties {inputs [
     {
      inputName=None
      input {
       expression=&self:MaterialExpressionTextureCoordinate_0
       inputName=None
       mask=0
       maskA=0
       maskB=0
       maskG=0
       maskR=0
       outputIndex=0
      }
     }
     {
      inputName=None
      input {
       expression=&self:MaterialExpressionTextureCoordinate_1
       inputName=None
       mask=0
       maskA=0
       maskB=0
       maskG=0
       maskR=0
       outputIndex=0
      }
     }
     {
      inputName=None
      input {
       expression=&self:MaterialExpressionTextureCoordinate_2
       inputName=None
       mask=0
       maskA=0
       maskB=0
       maskG=0
       maskR=0
       outputIndex=0
      }
     }
     {
      inputName=None
      input {
       expression=&self:MaterialExpressionTextureCoordinate_3
       inputName=None
       mask=0
       maskA=0
       maskB=0
       maskG=0
       maskR=0
       outputIndex=0
      }
     }
     {
      inputName=None
      input {
       expression=&self:MaterialExpressionTextureCoordinate_4
       inputName=None
       mask=0
       maskA=0
       maskB=0
       maskG=0
       maskR=0
       outputIndex=0
      }
     }
     {
      inputName=None
      input {
       expression=&self:MaterialExpressionTextureCoordinate_5
       inputName=None
       mask=0
       maskA=0
       maskB=0
       maskG=0
       maskR=0
       outputIndex=0
      }
     }
     {
      inputName=None
      input {
       expression=&self:MaterialExpressionTextureCoordinate_6
       inputName=None
       mask=0
       maskA=0
       maskB=0
       maskG=0
       maskR=0
       outputIndex=0
      }
     }
     {
      inputName=None
      input {
       expression=&self:MaterialExpressionTextureCoordinate_7
       inputName=None
       mask=0
       maskA=0
       maskB=0
       maskG=0
       maskR=0
       outputIndex=0
      }
     }
    ]}
  }
  MaterialExpressionTextureCoordinate_0 {class=MaterialExpressionTextureCoordinate, properties {uTiling=2, vTiling=2}}
  MaterialExpressionTextureCoordinate_1 {class=MaterialExpressionTextureCoordinate, properties {coordinateIndex=1, uTiling=2, vTiling=2}}
  MaterialExpressionTextureCoordinate_2 {class=MaterialExpressionTextureCoordinate, properties {coordinateIndex=2, uTiling=2, vTiling=2}}
  MaterialExpressionTextureCoordinate_3 {class=MaterialExpressionTextureCoordinate, properties {coordinateIndex=3, uTiling=2, vTiling=2}}
  MaterialExpressionTextureCoordinate_4 {class=MaterialExpressionTextureCoordinate, properties {coordinateIndex=4, uTiling=2, vTiling=2}}
  MaterialExpressionTextureCoordinate_5 {class=MaterialExpressionTextureCoordinate, properties {coordinateIndex=5, uTiling=2, vTiling=2}}
  MaterialExpressionTextureCoordinate_6 {class=MaterialExpressionTextureCoordinate, properties {coordinateIndex=6, uTiling=2, vTiling=2}}
  MaterialExpressionTextureCoordinate_7 {class=MaterialExpressionTextureCoordinate, properties {coordinateIndex=7, uTiling=2, vTiling=2}}
  MaterialExpressionTextureSample_0 {
   class=MaterialExpressionTextureSample
   inputs [
    UVs {from=MaterialExpressionSwitch_0}
    Tex {from=""}
    Apply View MipBias {from=""}
   ]
   properties {texture=&/T_CheckerBoard}
  }
  MaterialExpressionVectorParameter_0 {class=MaterialExpressionVectorParameter, properties {parameterName=Color, defaultValue=<rgba 1 1 1 1>}}
 }
 properties {
  bUsedWithInstancedStaticMeshes=true
  bUsedWithNanite=true
  editorOnlyData="<self>:PreviewMaterial_0EditorOnlyData"
  usedShadingModels=Default Lit
 }
}
