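# Two-sided translucent wireframe material for Dataflow editor visualization, using Default Lit shading and flagged for instanced static meshes and Nanite.
# A Color vector parameter is multiplied by per-instance custom data that defaults to white, so each instance can tint the shared color.
# A Transparency scalar is inverted through OneMinus to give an opacity-style value.
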Material /Engine/EditorMaterials/Dataflow/M_Dataflow_Color_Wireframe_TwoSided
graph {
 properties {
  bUsedWithInstancedStaticMeshes=true
  bUsedWithNanite=true
  blendMode=BLEND_Translucent
  twoSided=true
  usedShadingModels=Default Lit
  wireframe=true
 }
 nodes [
  Multiply(
   A=VectorParameter() {parameterName=Color, defaultValue=<rgba 1 1 1 1>}.RGB
   B=PerInstanceCustomData3Vector!(DefaultValue) {constDefaultValue=<rgba 1 1 1 0>}
  )
  OneMinus(None=ScalarParameter() {parameterName=Transparency, sliderMax=1})
 ]
}
