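# An unlit, two-sided editor material that previews a texture for Dataflow, with controls for channel masking and greyscale display.
# The Texture parameter sample is multiplied by the ColorMask vector to filter or tint its channels.
# A GreyScaleFactor scalar blends that masked color toward the texture's unmasked red channel, so greyscale mode shows red only rather than luminance.
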Material /Engine/EditorMaterials/Dataflow/M_Dataflow_TexturePreview
graph {
 properties {
  bUsedWithMeshParticles=true
  pixelDepthOffsetMode=PDOM_Legacy
  previewMesh=/Engine/Transient.MaterialEditorMeshComponent_3
  shadingModel=MSM_Unlit
  twoSided=true
  usedShadingModels=Unlit
 }
 nodes [
  n1 = TextureSampleParameter2D() {parameterName=Texture, texture=/Engine/EngineMaterials/DefaultDiffuse}
  LinearInterpolate(
   A=Multiply(A=@n1.RGBA, B=VectorParameter() {parameterName=ColorMask, defaultValue=<rgba 1 1 1 1>}.RGBA)
   B=@n1.R
   Alpha=ScalarParameter() {parameterName=GreyScaleFactor}
  )
 ]
}
